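; EQS context for the CoffeePuff ability set that supplies the teleport-punch destination as a single location to environment queries.
; It overrides ProvideSingleLocation, fetches the querier actor's blackboard and returns the vector stored under the hardcoded key CoffeePuff_TeleportPunchDestination.
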(fn ProvideSingleLocation
  (params (QuerierObject :type Object%20Reference) (QuerierActor :type Actor%20Object%20Reference))
  (return (AI|Components|Blackboard|GetValueasVector (AI|GetBlackboard QuerierActor) (Utilities|Name|MakeLiteralName "CoffeePuff_TeleportPunchDestination"))))
