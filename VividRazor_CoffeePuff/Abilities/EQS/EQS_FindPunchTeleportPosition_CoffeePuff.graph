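# EQS query that finds a teleport position for the CoffeePuff punch ability near the AI's current target.
# A two-ring, 180-degree donut aligned to the target's facing generates nav-projected points, excluding water.
# A pathfinding test checks the points against the target with a water-avoiding filter.
# Two dot tests against the target's facing reject points dead ahead or too far sideways, leaving front-diagonal spots.
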EnvQuery /VividRazor_CoffeePuff/Abilities/EQS/EQS_FindPunchTeleportPosition_CoffeePuff
options [{
  generator {
   name=EnvQueryGenerator_Donut_0
   class=EnvQueryGenerator_Donut
   properties {
    arcAngle=180
    bUseSpiralPattern=false
    center=&FortAthenaAIBot_EQSQueryContext_CurrentTarget
    innerRadius=500
    numberOfRings=2
    outerRadius=550
    pointsPerRing=25
    arcDirection {dirMode=Rotation, rotation=&FortAthenaAIBot_EQSQueryContext_CurrentTarget}
    projectionData {
     extentX=0
     navigationFilter=&NavFilter_NPC_ExcludeAllWater_C
     postProjectionVerticalOffset=0
     serializedChannel=ECC_GameTraceChannel11
     traceChannel=TraceTypeQuery10
    }
   }
  }
  tests [
   EnvQueryTest_Pathfinding_2 {
    class=EnvQueryTest_Pathfinding
    properties {
     context=&FortAthenaAIBot_EQSQueryContext_CurrentTarget
     filterClass=&NavFilter_Phoebe_AvoidWater_C
    }
   }
   EnvQueryTest_Dot_3 {
    class=EnvQueryTest_Dot
    properties {
     clampMaxType=None
     clampMinType=None
     floatValueMax=0.8746
     scoreClampMax=0.1
     scoreClampMin=0
     scoringFactor=10
     testComment=""
     testMode=Dot3D
     testPurpose=Filter
     lineA {rotation=&FortAthenaAIBot_EQSQueryContext_CurrentTarget}
     lineB {lineFrom=&FortAthenaAIBot_EQSQueryContext_CurrentTarget}
    }
   }
   EnvQueryTest_Dot_4 {
    class=EnvQueryTest_Dot
    properties {
     clampMaxType=None
     clampMinType=None
     filterType=Minimum
     floatValueMin=0.5
     scoreClampMax=0.1
     scoreClampMin=0
     scoringFactor=10
     testComment=""
     testMode=Dot3D
     lineA {rotation=&FortAthenaAIBot_EQSQueryContext_CurrentTarget}
     lineB {lineFrom=&FortAthenaAIBot_EQSQueryContext_CurrentTarget}
    }
   }
  ]
 }]
