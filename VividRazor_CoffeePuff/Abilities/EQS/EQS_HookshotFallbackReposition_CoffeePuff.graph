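# EQS query that picks a fallback reposition point for the hookshot ability, judging by its name, for an AI bot relative to its current target.
# A half-donut around the querier, oriented by the target's rotation, generates navmesh points filtered by water-avoiding pathfinding and a trace to the target.
# Dot tests constrain angles, favoring points in front of the target.
# A distance band filters points and scores them around a mid-range reference.
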EnvQuery /VividRazor_CoffeePuff/Abilities/EQS/EQS_HookshotFallbackReposition_CoffeePuff
options [{
  generator {
   name=EnvQueryGenerator_Donut_0
   class=EnvQueryGenerator_Donut
   properties {
    arcAngle=180
    bDefineArc=true
    bUseSpiralPattern=false
    center=&EnvQueryContext_Querier
    innerRadius=350
    numberOfRings=4
    outerRadius=1000
    pointsPerRing=25
    arcDirection {rotation=&FortAthenaAIBot_EQSQueryContext_CurrentTarget}
    projectionData {
     extentX=0
     navigationFilter=/Game/Athena/AI/Phoebe/NavFilter/NavFilter_Phoebe.NavFilter_Phoebe_C
     postProjectionVerticalOffset=0
     serializedChannel=ECC_GameTraceChannel11
     traceChannel=TraceTypeQuery10
    }
   }
  }
  tests [
   EnvQueryTest_Pathfinding_0 {
    class=EnvQueryTest_Pathfinding
    properties {filterClass=&NavFilter_Phoebe_AvoidWater_C}
   }
   EnvQueryTest_Dot_0 {
    class=EnvQueryTest_Dot
    properties {
     clampMaxType=None
     clampMinType=None
     floatValueMax=0.8104
     floatValueMin=-0.1051
     scoreClampMax=0.1
     scoreClampMin=0
     scoringFactor=10
     testComment=""
     testMode=Dot3D
     testPurpose=Filter
     lineA {rotation=&EnvQueryContext_Item}
     lineB {lineFrom=&FortAthenaAIBot_EQSQueryContext_CurrentTarget}
    }
   }
   EnvQueryTest_Dot_1 {
    class=EnvQueryTest_Dot
    properties {
     clampMaxType=None
     clampMinType=None
     filterType=Minimum
     floatValueMin=0.5
     scoreClampMax=0.1
     scoreClampMin=0
     scoringFactor=10
     testComment=""
     testMode=Dot3D
     lineA {rotation=&FortAthenaAIBot_EQSQueryContext_CurrentTarget}
     lineB {lineFrom=&FortAthenaAIBot_EQSQueryContext_CurrentTarget}
    }
   }
   EnvQueryTest_Distance_1 {
    class=EnvQueryTest_Distance
    properties {
     bDefineReferenceValue=true
     clampMaxType=SpecifiedValue
     clampMinType=SpecifiedValue
     distanceTo=&FortAthenaAIBot_EQSQueryContext_CurrentTarget
     filterType=Range
     floatValueMax=2000
     floatValueMin=500
     multipleContextScoreOp=Multiply
     normalizationType=RelativeToScores
     referenceValue=1250
     scoreClampMax=3814.9326
     scoreClampMin=616.4316
     scoringEquation=Square
     scoringFactor=1
     testComment=""
     testMode=Distance3D
     testPurpose=FilterAndScore
    }
   }
   EnvQueryTest_Trace_1 {
    class=EnvQueryTest_Trace
    properties {
     context=&FortAthenaAIBot_EQSQueryContext_CurrentTarget
     scoringFactor=1
     testComment=""
     testPurpose=Filter
     contextHeightOffset {
      dataBinding=&self:EnvQueryTest_Trace_1.FortAIDataProvider_ScalableFloat_1
      dataField=FloatValue
      defaultValue=0
     }
     itemHeightOffset {
      dataBinding=&self:EnvQueryTest_Trace_1.FortAIDataProvider_ScalableFloat_0
      dataField=FloatValue
      defaultValue=0
     }
    }
   }
  ]
 }]
