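# EQS query that picks a teleport position for the CoffeePuff AI's punch ability, relative to its current target.
# A two-ring half-donut around a custom context, oriented by target rotation and with data-bound radii, generates nav-projected candidates.
# A water-avoiding pathfinding test against the target checks candidates.
# Two dot tests on target facing together keep a front-diagonal band, with the second also scoring.
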EnvQuery /VividRazor_CoffeePuff/Abilities/EQS/EQS_PunchTeleportPosition_CoffeePuff
options [{
  generator {
   name=EnvQueryGenerator_Donut_0
   class=EnvQueryGenerator_Donut
   properties {
    arcAngle=180
    bUseSpiralPattern=false
    center=&/EQSC_PunchTeleportPosition_CoffeePuff.EQSC_PunchTeleportPosition_CoffeePuff_C
    numberOfRings=2
    pointsPerRing=25
    arcDirection {dirMode=Rotation, rotation=&FortAthenaAIBot_EQSQueryContext_CurrentTarget}
    innerRadius {
     dataBinding=&self:EnvQueryGenerator_Donut_0.FortAIDataProvider_ScalableFloat_0
     dataField=FloatValue
     defaultValue=200
    }
    outerRadius {
     dataBinding=&self:EnvQueryGenerator_Donut_0.FortAIDataProvider_ScalableFloat_1
     dataField=FloatValue
     defaultValue=230
    }
    projectionData {
     extentX=0
     navigationFilter=/Game/Athena/AI/Phoebe/NavFilter/NavFilter_Phoebe.NavFilter_Phoebe_C
     postProjectionVerticalOffset=0
     projectUp=60
     serializedChannel=ECC_GameTraceChannel11
     traceChannel=TraceTypeQuery10
    }
   }
  }
  tests [
   EnvQueryTest_Pathfinding_2 {
    class=EnvQueryTest_Pathfinding
    properties {
     context=&FortAthenaAIBot_EQSQueryContext_CurrentTarget
     filterClass=&NavFilter_Phoebe_AvoidWater_C
    }
   }
   EnvQueryTest_Dot_3 {
    class=EnvQueryTest_Dot
    properties {
     clampMaxType=None
     clampMinType=None
     floatValueMax=0.7051
     floatValueMin=-0.8949
     scoreClampMax=0.1
     scoreClampMin=0
     scoringFactor=10
     testComment=""
     testMode=Dot3D
     testPurpose=Filter
     lineA {rotation=&FortAthenaAIBot_EQSQueryContext_CurrentTarget}
     lineB {lineFrom=&FortAthenaAIBot_EQSQueryContext_CurrentTarget}
    }
   }
   EnvQueryTest_Dot_4 {
    class=EnvQueryTest_Dot
    properties {
     clampMaxType=None
     clampMinType=None
     filterType=Minimum
     floatValueMin=0.5
     scoreClampMax=0.1
     scoreClampMin=0
     scoringFactor=10
     testComment=""
     testMode=Dot3D
     lineA {rotation=&FortAthenaAIBot_EQSQueryContext_CurrentTarget}
     lineB {lineFrom=&FortAthenaAIBot_EQSQueryContext_CurrentTarget}
    }
   }
  ]
 }]
