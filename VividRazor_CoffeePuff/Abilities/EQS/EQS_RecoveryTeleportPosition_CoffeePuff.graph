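# Environment query for CoffeePuff that picks a recovery teleport position around the punch teleport position context.
# A donut generator builds a spiral-patterned half-ring of points around that context, oriented by the current target's rotation and projected using the Phoebe nav filter.
# A random test varies the choice, and a pathfinding test checks paths against the same punch teleport context with a water-avoiding nav filter.
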EnvQuery /VividRazor_CoffeePuff/Abilities/EQS/EQS_RecoveryTeleportPosition_CoffeePuff
options [{
  generator {
   name=EnvQueryGenerator_Donut_0
   class=EnvQueryGenerator_Donut
   properties {
    arcAngle=180
    bUseSpiralPattern=true
    center=&/EQSC_PunchTeleportPosition_CoffeePuff.EQSC_PunchTeleportPosition_CoffeePuff_C
    innerRadius=300
    numberOfRings=4
    outerRadius=700
    pointsPerRing=25
    arcDirection {dirMode=Rotation, rotation=&FortAthenaAIBot_EQSQueryContext_CurrentTarget}
    projectionData {
     extentX=0
     navigationFilter=/Game/Athena/AI/Phoebe/NavFilter/NavFilter_Phoebe.NavFilter_Phoebe_C
     postProjectionVerticalOffset=0
     projectDown=100
     projectUp=100
     serializedChannel=ECC_GameTraceChannel11
     traceChannel=TraceTypeQuery10
    }
   }
  }
  tests [
   FortQueryTest_Random_0 {class=FortQueryTest_Random, properties {floatValueMax=1, scoringFactor=1, testPurpose=FilterAndScore}}
   EnvQueryTest_Pathfinding_2 {
    class=EnvQueryTest_Pathfinding
    properties {
     context=&EQSC_PunchTeleportPosition_CoffeePuff_C
     filterClass=&NavFilter_Phoebe_AvoidWater_C
    }
   }
  ]
 }]
